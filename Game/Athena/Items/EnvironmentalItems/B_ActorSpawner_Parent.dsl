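; Base environmental spawner that places actors in Athena matches from a fixed or weighted class.
; On the server, unless manager-controlled, it waits for playlist data and, if present, minigame start.
; Data-driven rows gate spawning: either one randomly delayed actor or several at EQS locations with optional surface alignment.
; It spawns gameplay, building or plain actors, calls an overridable post-spawn event, and can self-destruct after a random lifespan.
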(event EventBeginPlay
  (bind _self self)
  (bind _returnvalue (Minigame|GetMinigame _self :!class /Script/FortniteGame.MinigameBlueprintLibrary))
  (bind _bapplyfix (CallFunction|ShouldApplyWaitforPlaylistFix))
  ;[0.02316 1 0] On the server, once the playlist has been made check if the spawn row is enabled and if we pass the spawn chance. If so, start a delay with a random interval (for perf) then spawn the actor after.
  (if (Networking|IsServer)
    (if (not (Variables|Default|GetspawnManagerControlled?))
      (Class|FortAsyncActionCurrentPlaylistReady|PlaylistDataReadyAsync
        (:then
          ;[1 0.8972 0] Hold off spawning anything until the minigame starts. This is to keep it working as is in BR but allow for it to spawn at the right time in Creative Royale, as BR does not currently use the minigame.
          (if _bapplyfix
            _
            (else
              _)))
        (:OnReady
          (if _bapplyfix
            _)))
      (Utilities|IsValid _returnvalue
        (:Is%20Valid
          (Minigame|BindEventtoMinigameStartedDelegate _returnvalue (EventDispatchers|CreateEvent :!event Custom|TriggeredSpawnActivation)))
        (:Is%20Not%20Valid
          (CallFunction|TriggeredSpawnActivation))))))

(event EventDestroyed
  ;[1 0.008413 0] remove the bind when this inevitably is destroyed
  (|CallParentFunction)
  (CallFunction|UnbindMinigame))

(event EQS|EventPrepForPrevis
  ;[1 1 1] EQS Previs Setup
  (if (Variables|SpawnerVars|GetEQS_Enabled_MultiSpawn)
    (EQS|SetQueryTemplate (Variables|SpawnerVars|GetEQS_QueryTemplate))))

(event Custom|EQS%20Finished
  (params (QueryInstance :type EQS%20Query%20Instance%20Object%20Reference) (QueryStatus :type EEnvQueryStatus%20Enum :default "Processing"))
  ;[1 1 1] EQS for spawning actor
  (Development|PrintString (Utilities|String|EnumtoString QueryStatus))
  (bind (_resultlocations _returnvalue) (AI|EQS|GetQueryResultsasLocations QueryInstance))
  (Variables|Non-Tunables|SetLocationsToSpawn _resultlocations)
  (if _returnvalue
    (CallFunction|StartEQSSpawnLoop)
    (else
      (Development|PrintString "No Valid Locations from EQS Query")
      (CallFunction|DestroySpawner))))

(event Custom|TriggeredSpawnActivation
  (bind (_returnvalue_location _returnvalue_rotation _returnvalue_scale) (Math|Transform|BreakTransform (Transformation|GetActorTransform)))
  (CallFunction|UnbindMinigame)
  (if (Class|AthenaFunctionLibrary|RowtoBool (Variables|SpawnerVars|GetRow_Enabled))
    (if (Variables|SpawnerVars|GetEQS_Enabled_MultiSpawn)
      ;[1 1 1] EQS for spawning actor
      (bind _returnvalue (AI|EQS|RunEQSQuery (Variables|SpawnerVars|GetEQS_QueryTemplate) self "AllMatching"))
      (AI|EQS|BindEventtoOnQueryFinished _returnvalue (delegate Custom|EQS%20Finished))
      (else
        ;[1 1 1] Simple Non EQS spawner; just spawns an individual object using chance
        (if (Math|Random|RandomBoolwithWeight (Math|Float|Clamp(Float) (ScalableFloat|GetValueatLevelwithContext (Variables|SpawnerVars|GetRow_SpawnChance) (Variables|SpawnerVars|GetSpawnChanceLevel))))
          (Utilities|FlowControl|Delay (Math|Random|RandomFloatinRange (Variables|SpawnerVars|GetSpawnDelayMin) (Variables|SpawnerVars|GetSpawnDelayMax)))
          (CallFunction|SpawnActor (Math|Transform|MakeTransform (+ (Variables|SpawnerVars|GetSpawnLocationOffset) _returnvalue_location) _returnvalue_rotation (Math|Vector|SelectVector (Variables|/Script/Engine.SceneComponent|GetRelativeScale3D (Variables|Default|GetPreVisMesh)) _returnvalue_scale (and (Variables|SpawnerVars|GetUsePrevisMeshScale) (Utilities|IsValid (Variables|Default|GetPreVisMesh) :!class /Script/Engine.KismetSystemLibrary)))))
          (CallFunction|DestroySpawner)
          (else
            _))))
    (else
      _))
  ;[1 1 1] If Destroy after spawn is true set random life span.
  (if (Variables|SpawnerVars|GetDestroyOnFailSpawnChance)
    (Actor|SetLifeSpan (Math|Random|RandomFloatinRange (Variables|SpawnerVars|GetDestroySelfLifeMin) (Variables|SpawnerVars|GetDestroySelfLifeMax)))))

(event Custom|DestroySpawner
  ;[1 1 1] If Destroy after spawn is true set random life span.
  (if (Variables|SpawnerVars|GetDestroyAfterSpawn)
    (Actor|SetLifeSpan (Math|Random|RandomFloatinRange (Variables|SpawnerVars|GetDestroySelfLifeMin) (Variables|SpawnerVars|GetDestroySelfLifeMax)))))

(event Custom|StartEQSSpawnLoop
  (bind _locationstospawn (Variables|Non-Tunables|GetLocationsToSpawn))
  (bind _spawnlocationoffset (Variables|SpawnerVars|GetSpawnLocationOffset))
  ;[1 1 1] StartEQSSpawn - Loop through and spawn the fishing pools. Run a test to see if we should spawn this one. If we fail, still increment but don't spawn
  ;;[1 1 1] Determine whether this Cluster will spawn at all based on Spawn Chance
  (if (Math|Random|RandomBoolwithWeight (Math|Float|Clamp(Float) (ScalableFloat|GetValueatLevelwithContext (Variables|SpawnerVars|GetRow_SpawnChance))))
    ;;[1 1 1] Calcualte the # of actors to spawn at this cluster
    (bind _output_get (Variables|Non-Tunables|SetcachedSpawnAmount (Math|Integer|Clamp(Integer) (CallFunction|GetSpawnAmountforEQS) 0 (Utilities|Array|Length (Variables|Non-Tunables|GetLocationsToSpawn)))))
    (for _index (range _output_get)
      (bind (_outhit _returnvalue) (Collision|LineTraceByChannel (+ (+ (Utilities|Array|Get(aref) _locationstospawn _index) _spawnlocationoffset) "0.000000,0.000000,50.000000") (+ (+ (Utilities|Array|Get(aref) _locationstospawn _index) _spawnlocationoffset) "0.000000,0.000000,-50.000000")))
      (CallFunction|SpawnActor (Math|Transform|MakeTransform (+ (Utilities|Array|Get(aref) _locationstospawn _index) _spawnlocationoffset) (Math|Rotator|SelectRotator (Math|Rotator|MakeRotfromZX (.Normal (Collision|BreakHitResult _outhit)) (Math|Vector|SelectVector (Math|Vector|RotateVectorAroundAxis "1.000000,0.000000,0.000000" (Math|Random|RandomFloatinRange -1.8e+02 1.8e+02) (.Normal (Collision|BreakHitResult _outhit))) (Math|Vector|RotateVectorAroundAxis "1.000000,0.000000,0.000000" 9e+01 (.Normal (Collision|BreakHitResult _outhit))) (Variables|SpawnerVars|GetEQS_RandomYaw))) (Math|Rotator|SelectRotator (Math|Rotator|MakeRotfromZX (Math|Vector|RotateVectorAroundAxis "0.000000,0.000000,1.000000" 0.0 (Math|Rotator|GetRotationXVector (Transformation|GetActorRotation))) (Math|Rotator|GetRotationXVector (Transformation|GetActorRotation))) (Transformation|GetActorRotation) (Variables|SpawnerVars|GetEQS_RandomYaw)) (and _returnvalue (Variables|SpawnerVars|GetEQS_AlignToSurfaceNormal))) (* "1.000000,1.000000,1.000000" (Math|Random|RandomFloatinRange (Variables|SpawnerVars|GetEQS_ScaleMIN) (Variables|SpawnerVars|GetEQS_ScaleMAX))))))
    (CallFunction|DestroySpawner)))

(event Custom|SpawnActor
  (params (New%20Transform :type Transform))
  ;[1 1 1] If the Actor is a valid class spawn it. If its BGA spawn using BGA specific spawn method. Finally call a post-spawn event that can be overidden in children
  (bind _returnvalue (CallFunction|GetActorClasstoSpawn))
  (if (Utilities|IsValidClass _returnvalue)
    (if (not (Variables|Default|GetSkipTrySpawnAsBGA))
      (bind _asbuilding_gameplay_actor (Utilities|Casting|CastToBuildingGameplayActorClass _returnvalue)
        (:then
          (bind _returnvalue_2 (Building|SpawnBuildingGameplayActor _asbuilding_gameplay_actor New%20Transform (Game|GetGameMode) true))
          (CallFunction|ServerActorSpawned _returnvalue_2))
        (:CastFailed
          _)))
    (bind _asbuilding_actor (Utilities|Casting|CastToBuildingActorClass _returnvalue)
      (:then
        (bind _returnvalue_3 (Building|SpawnBuildingActor _asbuilding_actor New%20Transform))
        (Building|InitializeKismetSpawnedBuildingActor _returnvalue_3 0 0 true 0 false)
        (CallFunction|ServerActorSpawned _returnvalue_3))
      (:CastFailed
        (bind _returnvalue_4 (Game|SpawnActorfromClass _returnvalue New%20Transform "AlwaysSpawn" "SelectDefaultAtRuntime" 0 0))
        (CallFunction|ServerActorSpawned _returnvalue_4)))))

(event Custom|ServerActorSpawned
  (params (Actor :type Actor%20Object%20Reference)))

(fn BuildDataRegistryResolverScope
  (bind _self self)
  (bind _returnvalue (Minigame|GetMinigame _self :!class /Script/FortniteGame.MinigameBlueprintLibrary))
  ;[1 0.8475 0.2776] Gives the name of the volume as a scope to the data registry. Needed for calls to GetValueWithContext to work
  (Utilities|IsValid _returnvalue
    (:Is%20Valid
      (return (Utilities|Array|MakeArray (SpatialGameplay|GetSpatialGameplayInterfaceFNameforActor _returnvalue)) 0 false))))

(fn GetActorClassToSpawn
  (if (Variables|SpawnerVars|GetUseWeightedSpawnClassList)
    (bind _returnvalue (Class|WeightedActorClassListLibrary|SelectRandomActorTypefromList (Variables|SpawnerVars|GetWeightedSpawnClassList)))
    (return _returnvalue)
    (else
      (return (Variables|SpawnerVars|GetClassToSpawn)))))

(fn GetCacheContextOverride
  ;[0.4163 1 0.9801] We don't need to return anything here since we don't need to override the Cache Context.
  (return 0))

(fn GetSpawnAmountForEQS
  ;[1 1 1] Use spawn min max values and density to find the amount of actors to try and spawn via EQS
  (return (Math|Float|Truncate (* (Math|Integer|SelectInt (Math|Random|RandomIntegerinRange (Math|Float|Truncate (ScalableFloat|GetValueatLevel (Variables|SpawnerVars|GetRow_EQS_NumtoSpawnMIN))) (Math|Float|Truncate (ScalableFloat|GetValueatLevel (Variables|SpawnerVars|GetRow_EQS_NumtoSpawnMAX)))) (Math|Random|RandomIntegerinRange (Variables|SpawnerVars|GetEQS_NumtoSpawnMIN) (Variables|SpawnerVars|GetEQS_NumtoSpawnMAX)) (Variables|SpawnerVars|GetUseRowsForMinMaxEQS)) (Math|Float|Clamp(Float) (ScalableFloat|GetValueatLevel (Variables|SpawnerVars|GetRow_EQS_SpawnDensity)))))))

(fn ShouldApplyWaitForPlaylistFix
  (bind _row_applywaitforplaylistfix (Variables|Default|GetRow_ApplyWaitForPlaylistFix))
  (if (ScalableFloat|IsSettingCurveValid _row_applywaitforplaylistfix)
    (return (ScalableFloat|GetValueasBool _row_applywaitforplaylistfix))
    (else
      (return false))))

(fn UnbindMinigame
  (bind _self self)
  (bind _returnvalue (Minigame|GetMinigame _self :!class /Script/FortniteGame.MinigameBlueprintLibrary))
  ;[1 0.6819 0] Unbind the TriggeredSpawnActivation from the Minigame Started Delegate
  (Utilities|IsValid _returnvalue
    (:Is%20Valid
      (Minigame|UnbindEventfromMinigameStartedDelegate _returnvalue (EventDispatchers|CreateEvent :!event TriggeredSpawnActivation)))))

(fn ConstructionScript
  ;[1 1 1] Toggle Editor Text number ON for EQS based
  (Rendering|SetVisibility (Variables|Default|GetEditorText-AmountToSpawn) (Variables|SpawnerVars|GetEQS_Enabled_MultiSpawn) false)
  ;[1 1 1] Text shows the range of # to spawn (min - max)
  (Rendering|Components|TextRender|SetText (Variables|Default|GetEditorText-AmountToSpawn) (Utilities|Text|ToText(String) (Utilities|String|Append (Utilities|String|ToString(Integer) (Variables|SpawnerVars|GetEQS_NumtoSpawnMIN)) " - " (Utilities|String|ToString(Integer) (Variables|SpawnerVars|GetEQS_NumtoSpawnMAX))))))
